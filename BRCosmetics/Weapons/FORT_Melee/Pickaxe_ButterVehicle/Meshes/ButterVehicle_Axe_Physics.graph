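# Physics asset for the ButterVehicle pickaxe, giving physics bodies to a chain of "dyn_Bread" bones on the weapon mesh.
# Five identical capsules form a serial chain from a tiny capsule on dyn_Bread_MSTR, each constrained to the previous bone.
# Every constraint all but locks swing1 and twist and allows limited swing2, so the chain bends like linked hinges, each on a single axis.
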PhysicsAsset /BRCosmetics/Weapons/FORT_Melee/Pickaxe_ButterVehicle/Meshes/ButterVehicle_Axe_Physics
bodies {
 dyn_Bread_01_jnt [{shapeType=Capsule, length=5, radius=3, shapeName=dyn_Bread_01_jnt_capsule, center=<5 0 0>, rotation=<rot 90 0 0>}]
 dyn_Bread_02_jnt [{shapeType=Capsule, length=5, radius=3, shapeName=dyn_Bread_02_jnt_capsule, center=<5 0 0>, rotation=<rot 90 0 0>}]
 dyn_Bread_03_jnt [{shapeType=Capsule, length=5, radius=3, shapeName=dyn_Bread_03_jnt_capsule, center=<5 0 0>, rotation=<rot 90 0 0>}]
 dyn_Bread_04_jnt [{shapeType=Capsule, length=5, radius=3, shapeName=dyn_Bread_04_jnt_capsule, center=<5 0 0>, rotation=<rot 90 0 0>}]
 dyn_Bread_05_jnt [{shapeType=Capsule, length=5, radius=3, shapeName=dyn_Bread_05_jnt_capsule, center=<5 0 0>, rotation=<rot 90 0 0>}]
 dyn_Bread_MSTR [{shapeType=Capsule, length=0.505, radius=0.505, shapeName=dyn_Bread_MSTR_capsule, rotation=<rot 0 -90 0>}]
}
constraints [
 {bone1Name=dyn_Bread_05_jnt, bone2Name=dyn_Bread_04_jnt, swing1LimitDegrees=0.01, swing2LimitDegrees=25, twistLimitDegrees=0.01}
 {bone1Name=dyn_Bread_01_jnt, bone2Name=dyn_Bread_MSTR, swing1LimitDegrees=0.01, swing2LimitDegrees=25, twistLimitDegrees=0.01}
 {bone1Name=dyn_Bread_02_jnt, bone2Name=dyn_Bread_01_jnt, swing1LimitDegrees=0.01, swing2LimitDegrees=25, twistLimitDegrees=0.01}
 {bone1Name=dyn_Bread_03_jnt, bone2Name=dyn_Bread_02_jnt, swing1LimitDegrees=0.01, swing2LimitDegrees=25, twistLimitDegrees=0.01}
 {bone1Name=dyn_Bread_04_jnt, bone2Name=dyn_Bread_03_jnt, swing1LimitDegrees=0.01, swing2LimitDegrees=25, twistLimitDegrees=0.01}
]
