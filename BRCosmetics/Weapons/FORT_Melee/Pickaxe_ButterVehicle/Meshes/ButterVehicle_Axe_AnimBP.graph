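# Animation Blueprint for the ButterVehicle pickaxe that adds simulated secondary motion to the axe mesh.
# The AnimGraph runs the input pose through a RigidBody node using an override physics asset, simulated relative to the two-handed handle bone with zero gravity and damped, clamped component motion.
# Each update, the EventGraph sets a float from bIsPlayingMeleeAnim and scales it into the node's alpha, so physics applies partially during melee animations and not otherwise.
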AnimBlueprint /BRCosmetics/Weapons/FORT_Melee/Pickaxe_ButterVehicle/Meshes/ButterVehicle_Axe_AnimBP
target_skeleton=&/ButterVehicle_Axe_Skeleton
graphs [
 AnimGraph {nodes [
   AnimGraphNode_SubInput_0 = Input Pose()
   n1 = RigidBody!(
    ComponentPose=Local To Component(LocalPose=@AnimGraphNode_SubInput_0)
    Alpha=Get Is Melee Anim Alpha() -> (Is Melee Anim Alpha:double)
    OverrideWorldGravity:Vector="0,0,0"
    ExternalForce
   ) -> (Pose) {
    bClampLinearTranslationLimitToRefPose=true
    bOverrideWorldGravity=true
    overridePhysicsAsset=&/ButterVehicle_Axe_Physics
    simulationSpace=BaseBoneSpace
    alphaBoolBlend {blendOption=Linear}
    alphaScaleBiasClamp {interpSpeedIncreasing=0.5}
    baseBoneRef {boneName=Melee_TwoHanded_Handle}
    componentAppliedLinearAccClamp=<3000 3000 3000>
    componentLinearAccScale=<0.4 0.3 0.3>
    componentLinearVelScale=<0.4 0.3 0.3>
    simSpaceSettings {
     maxAngularAcceleration=1000
     maxAngularVelocity=3
     maxLinearAcceleration=700
     maxLinearVelocity=500
     velocityScaleZ=0
     externalLinearDragV=<0.2 0.2>
    }
   }
   Output Pose(Result=Component To Local(ComponentPose=@n1)) {layerGroup=None}
  ]}
 EventGraph {nodes [
   Event BlueprintUpdateAnimation()
   > n1 = Sequence()
   > n2 = Branch(Condition=Get bIsPlayingMeleeAnim())
   > Set Is Melee Anim Float(Is Melee Anim Float:float=0.25) -> (Output_Get:float)
   Set Is Melee Anim Float(execute=@n2.else, Is Melee Anim Float=0)
   Set Is Melee Anim Alpha(
    execute=@n1.then_1
    Is Melee Anim Alpha:double=float * float!(A=0.75, B=1, C:float=Get Is Melee Anim Float() -> (Is Melee Anim Float:float)) -> (ReturnValue) {is=title K2Node_PromotableOperator}
   ) -> (Output_Get:double)
   TryGetPawnOwner()
  ]}
]
variables [
 Is Melee Anim Alpha {type=number}
 Is Melee Anim Float {type=number}
]
